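# Sound cue for the Giant Space Heater environment actor, playing a rising warning sound before it explodes.
# A Random node picks one of two equally weighted rise waves and feeds a Modulator left at default settings.
# It uses the enemy attack sound class, a boosted volume and a two-voice concurrency limit.
# Overridden attenuation adds distance low-pass filtering, listener focus and a send to the distant reverb submix.
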SoundCue /Game/Sounds/Asteria/Environment_Actors/GiantSpaceHeater/Pre_Explode_GiantSpaceHeater_Cue
root=Modulator_1
nodes [
 {id=Modulator_1, children [Random_0]}
 {id=Random_0, children [WavePlayer_1,WavePlayer_2], properties {weights [1,1]}}
 {id=WavePlayer_1, properties {soundWaveAssetPtr=&/Pre_Exp_Rise_GiantSpaceHeater_01}}
 {id=WavePlayer_2, properties {soundWaveAssetPtr=&/Pre_Exp_Rise_GiantSpaceHeater_02}}
]
properties {
 attenuationSettings=/Game/Sounds/Attenuation/Fort_Monster_Husk_Att
 bOverrideAttenuation=true
 bOverrideConcurrency=true
 maxDistance=4000
 soundClassObject=/Game/Sounds/FortSoundClassesAndModes/Fort_Enemy_Attack
 volumeMultiplier=1.5
 attenuationOverrides {
  bAttenuateWithLPF=true
  bEnableFocusInterpolation=true
  bEnableListenerFocus=true
  bEnableSubmixSends=true
  falloffDistance=2500
  focusAttackInterpSpeed=0.5
  focusReleaseInterpSpeed=0.5
  lPFRadiusMax=3000
  lPFRadiusMin=1000
  nonFocusVolumeAttenuation=0.5
  stereoSpread=1000
  attenuationShapeExtents {x=1500}
  submixSendSettings [{
    bEnableHPFCutoff=false
    bEnableLPFCutoff=false
    disableManualSendClamp=false
    hPFCutoff=20
    lPFCutoff=20000
    maxSendDistance=7500
    maxSendLevel=0.25
    minSendDistance=0
    minSendLevel=0
    sendLevel=0.2
    sendLevelControlMethod=Linear
    soundSubmix=&Fortnite_Reverb_Distant
   }]
 }
 concurrencyOverrides {maxCount=2}
}
