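# Looping spot ambience cue for the giant space heater environment actor, on the Athena ambient sound class with a small concurrency cap.
# A random node picks one of three equally weighted loops, then a modulator at default settings varies it.
# An attenuation node adds distance low-pass filtering, listener focus and occlusion, with reverb send off.
# The root quality-level node routes the first two slots to that chain and leaves the other two empty.
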SoundCue /Game/Sounds/Asteria/Environment_Actors/GiantSpaceHeater/Spot_GiantSpaceHeater_Cue
root=QualityLevel_0
nodes [
 {
  id=Attenuation_1
  children [Modulator_0]
  properties {
   bOverrideAttenuation=true
   attenuationOverrides {
    bAttenuateWithLPF=true
    bEnableFocusInterpolation=true
    bEnableListenerFocus=true
    bEnableReverbSend=false
    falloffDistance=1200
    focusReleaseInterpSpeed=0.75
    lPFFrequencyAtMax=2000
    lPFRadiusMax=3000
    lPFRadiusMin=500
    nonFocusVolumeAttenuation=0.5
    occlusionInterpolationTime=1
    occlusionLowPassFilterFrequency=2000
    occlusionTraceChannel=ECC_GameTraceChannel11
    occlusionVolumeAttenuation=0.5
    stereoSpread=512
   }
  }
 }
 {id=Modulator_0, children [Random_1]}
 {id=QualityLevel_0, children [Attenuation_1, Attenuation_1, null, null]}
 {id=Random_1, children [WavePlayer_1,WavePlayer_3,WavePlayer_4], properties {weights [1,1,1]}}
 {
  id=WavePlayer_1
  properties {bLooping=true, soundWaveAssetPtr=&/Spot_GiantSpaceHeater_03_Loop}
 }
 {
  id=WavePlayer_3
  properties {bLooping=true, soundWaveAssetPtr=&/Spot_GiantSpaceHeater_01_Loop}
 }
 {
  id=WavePlayer_4
  properties {bLooping=true, soundWaveAssetPtr=&/Spot_GiantSpaceHeater_02_Loop}
 }
]
properties {
 bOverrideConcurrency=true
 maxDistance=1600
 soundClassObject=&Fort_Ambient_Athena
 volumeMultiplier=0.301
 attenuationOverrides {attenuationShapeExtents {x=0}}
 concurrencyOverrides {maxCount=3}
}
