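# Sound cue for the looping ambient hum of the Giant Space Heater environment actor, in the Athena ambient sound class.
# A looping wave runs through a modulator and two endlessly looping envelopers: a 60-second pitch curve and a 20-second volume swell.
# Cone attenuation adds distance low-pass filtering, listener focus and occlusion.
# The quality switch feeds only its first two slots, and the cue caps concurrency and sends to an exterior ambience bus.
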SoundCue /Game/Sounds/Asteria/Environment_Actors/GiantSpaceHeater/Spot_GiantSpaceHeater_Hum_Cue
root=QualityLevel_0
nodes [
 {
  id=Attenuation_1
  children [Enveloper_4]
  properties {
   bOverrideAttenuation=true
   attenuationOverrides {
    attenuationShape=Cone
    bAttenuateWithLPF=true
    bEnableFocusInterpolation=true
    bEnableListenerFocus=true
    bEnableReverbSend=false
    bEnableSubmixSends=true
    falloffDistance=1200
    focusReleaseInterpSpeed=0.75
    focusVolumeAttenuation=1.25
    lPFFrequencyAtMax=2000
    lPFRadiusMax=3000
    lPFRadiusMin=500
    nonFocusVolumeAttenuation=0.1
    occlusionInterpolationTime=1
    occlusionLowPassFilterFrequency=2000
    occlusionTraceChannel=ECC_GameTraceChannel11
    occlusionVolumeAttenuation=0.5
    stereoSpread=100
    attenuationShapeExtents {y=80}
   }
  }
 }
 {
  id=Enveloper_0
  children [Modulator_1]
  properties {
   bLoop=true
   bLoopIndefinitely=true
   loopEnd=60
   pitchCurve {editorCurveData {keys [
      {interpMode=RCIM_Cubic, time=0}
      <at 15 0.95 RCIM_Cubic>
      {arriveTangent=0.0091, interpMode=RCIM_Cubic, leaveTangent=0.0091, tangentMode=RCTM_User, time=30}
      {interpMode=RCIM_Cubic, time=45}
      {time=60}
     ]}}
   volumeCurve {editorCurveData {keys [<at 0 1>]}}
  }
 }
 {
  id=Enveloper_4
  children [Enveloper_0]
  properties {
   bLoop=true
   bLoopIndefinitely=true
   loopEnd=20
   pitchCurve {editorCurveData {keys [{time=0}]}}
   volumeCurve {editorCurveData {keys [
      {arriveTangent=-0.01, interpMode=RCIM_Cubic, time=0, value=1}
      <at 10 1.2 RCIM_Cubic>
      {interpMode=RCIM_Cubic, leaveTangent=0.01, time=20, value=1}
     ]}}
  }
 }
 {id=Modulator_1, children [WavePlayer_0]}
 {id=QualityLevel_0, children [Attenuation_1, Attenuation_1, null, null]}
 {
  id=WavePlayer_0
  properties {bLooping=true, soundWaveAssetPtr=&/Spot_GiantSpaceHeater_Hum_Loop}
 }
]
properties {
 bOverrideConcurrency=true
 maxDistance=1600
 soundClassObject=&Fort_Ambient_Athena
 volumeMultiplier=1.103
 attenuationOverrides {attenuationShapeExtents {x=0}}
 busSends [{audioBus=null, minSendLevel=1, sendLevel=1, soundSourceBus=&AmbientAudio_Exterior_SourceBus}]
 concurrencyOverrides {maxCount=3}
}
